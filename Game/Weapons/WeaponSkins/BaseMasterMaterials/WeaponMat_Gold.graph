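# Gold master material for weapon skins, built as fully metallic Default Lit with material attributes.
# The gun master function supplies a specular mask that scales the gold Color1 tint, plus world position offset and masks that combine with a tiled noise map for roughness.
# The attributes, with a cloud normal map, pass through ElementalWeaponsFunction.
# An animated sparkle-normal network and the Lightning, Fire and Ice switches sit unconnected.
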Material /Game/Weapons/WeaponSkins/BaseMasterMaterials/WeaponMat_Gold
graph {
 properties {
  bUseMaterialAttributes=true
  bUsedWithMeshDeformer=true
  bUsedWithNanite=true
  bUsedWithSkeletalMesh=true
  numCustomizedUVs=5
  pixelDepthOffsetMode=PDOM_Legacy
  usedShadingModels=Default Lit
 }
 nodes [
  n1 = TextureSample() {samplerType=SAMPLERTYPE_LinearColor, texture=/Engine/EngineMaterials/RandomAngles}
  n2 = Multiply(A=Subtract(A=@n1.RGB) {constB=0.5}) {constB=2}
  n3 = Add(A=@n1.R, B=Multiply(A=Time()) {constB=0.1})
  n4 = Constant() {r=-0.1}
  n5 = Constant() {r=0.1}
  n6 = TextureCoordinate() {uTiling=64, vTiling=64}
  n7 = Divide(A=Add(A=Ceil(None=@n6)) {constB=0.5}) {constB=64}
  n8 = Normalize(VectorInput=LinearInterpolate(
    A=@n2
    B=TextureSampleParameter2D(UVs=@n7) {parameterName=normals, samplerType=SAMPLERTYPE_Normal, texture=/Game/Weapons/Textures/CloudImageNormals}.RGB
   ) {constAlpha=0.75})
  Add(
   A=RotateAboutAxis(
    NormalizedRotationAxis=Normalize(VectorInput=@n2)
    RotationAngle=MaterialFunctionCall!(Sine Phase=@n3, Value 1=@n4, Value 2=@n5) {materialFunction=/Engine/Functions/Engine_MaterialFunctions02/Utility/Sine_Remapped}.Remapped Sine
    PivotPoint=Constant()
    Position=@n8
   )
   B=@n8
  )
  Constant()
  n9 = VectorParameter() {parameterName=Color1, defaultValue=<rgba 0.72 0.3873 0 1>}
  n10 = MaterialFunctionCall!(UV 1 coordinates, MuzzleFalloff) {materialFunction=/Game/Packages/Fortnite_UniqueMaterials_1/Weapons/M_GunMasterShaderFunction}
  n11 = Multiply(A=Constant() {r=2}, B=@n10."Specular (_S R)")
  n12 = TextureSample(UVs=TextureCoordinate() {uTiling=3, vTiling=3}) {texture=&2ndInfectedMapLinear}
  n13 = Multiply(A=@n10."Customization Mask UV Islands (_CM B Bottom Right)", B=Multiply(A=OneMinus(None=@n12.R)) {constB=0.2})
  MaterialFunctionCall!(
   In=MakeMaterialAttributes!(
    BaseColor=Multiply(A=@n9.RGB, B=@n11)
    Metallic=Constant() {r=1}
    Specular
    Roughness=Multiply(A=@n13, B=@n10."Customization Crevasse (_CM G)")
    Anisotropy
    EmissiveColor
    Opacity
    OpacityMask
    Normal=TextureSampleParameter2D() {parameterName=normals, samplerType=SAMPLERTYPE_Normal, texture=&CloudImageNormals}.RGB
    Tangent
    WorldPositionOffset=@n10.World Position Offset
    SubsurfaceColor
    ClearCoat
    ClearCoatRoughness
    AmbientOcclusion
    Refraction
    CustomizedUVs_0
    CustomizedUVs_1
    CustomizedUVs_2
    CustomizedUVs_3
    CustomizedUVs_4
    CustomizedUVs_5
    CustomizedUVs_6
    CustomizedUVs_7
    PixelDepthOffset
    ShadingModel
    Displacement
   )
   Pre-Skinned Local Position
   Pre-Skinned Local Normal
  ) {materialFunction=/Game/Weapons/WeaponSkins/MaterialFunction/ElementalWeaponsFunction}
  StaticBoolParameter() {parameterName=Use Lightning}
  StaticBoolParameter() {parameterName=Use Fire}
  StaticBoolParameter() {parameterName=Use Ice}
 ]
}
